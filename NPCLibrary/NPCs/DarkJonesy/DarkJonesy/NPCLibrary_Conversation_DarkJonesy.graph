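# Conversation database for the DarkJonesy NPC, defining the dialogue that begins at its own entry tag.
# The entry leads to one speech node with a default line and contextual greetings chosen by curve-table priorities.
# These greetings check time since last match, a Tower Ruins visit, first-ever meeting, and high Bars count.
# The speech then branches through links to the shared Quests, Basic Services and Season 18 Fortnitemares conversations.
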ConversationDatabase /NPCLibrary/NPCs/DarkJonesy/DarkJonesy/NPCLibrary_Conversation_DarkJonesy
entry_points [{tag=Conversation.Entry.DarkJonesy, destinations [#1]}]
nodes [
 ConversationEntryPointNode_7 {
  id=#1
  class=ConversationEntryPointNode
  outputs [#2]
  properties {entryTag {tagName=Conversation.Entry.DarkJonesy}}
 }
 Speech_C_4 {
  id=#2
  class=Speech_C
  outputs [#3, #4, #5]
  properties {
   nodeName=""
   generalConfig {
    defaultMessage=The future used to be so clear...
    contextualMessages [
     {
      message="I should have foreseen your return. "
      contextRequirements [{requirement {class=FortControllerRequirement_CheckTimeSinceLastPlayedMatch}}]
      priority {curve {rowName=Tandem.Speech.Priority.WhileSincePlayed}}
     }
     {
      message=" How bad was it? The Cube site?"
      contextRequirements [{requirement {class=FortControllerRequirement_HasVisitedPOI, properties {pOITag {tagName=Athena.Location.UnNamedPOI.TowerRuins}}}}]
      priority {curve {rowName=Tandem.Speech.Priority.VisitedSpecificPOI}}
     }
     {
      message=You've come to help us. But I don't know how... or why...
      contextRequirements [{requirement {class=FortControllerRequirement_IsFirstEverConversationWithNPC}}]
      priority {curve {rowName=Tandem.Speech.Priority.FirstEverInteraction}}
     }
     {
      message=Bars won't help us out of this one.
      contextRequirements [{requirement {class=FortControllerRequirement_HasBars, properties {thresholdTestConfig {quantity {value=3000}}}}}]
      priority {curve {rowName=Tandem.Speech.Priority.HighTotalBars}}
     }
    ]
   }
  }
 }
 ConversationLinkNode_2 {
  id=#5
  class=ConversationLinkNode
  properties {remoteEntryTag {tagName=Conversation.Entry.Season18.Fortnitemares}}
 }
 ConversationLinkNode_0 {id=#4, class=ConversationLinkNode, properties {remoteEntryTag {tagName=Conversation.Entry.BasicServices}}}
 ConversationLinkNode_1 {id=#3, class=ConversationLinkNode, properties {remoteEntryTag {tagName=Conversation.Entry.NPCLibrary.Quests}}}
]
